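; Burst gameplay cue for a weapon bunker lock opening its door, playing the Tidal Crane bunker's door-open audio.
; After calling the parent burst handler, it casts the target's parent actor to BGA_BunkerDoor_TidalCrane.
; It then spawns two attached sounds on that door's Speaker: a scan speaker voice-over and an inner-lock sound.
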(fn OnBurst(Cascade)
  (params (MyTarget :type Actor%20Object%20Reference) (Parameters :type Gameplay%20Cue%20Parameters%20Structure%20%28by%20ref%29) (ParticleComponents :type Array%20of%20Cascade%20Particle%20System%20Component%20Object%20References) (AudioComponents :type Array%20of%20Audio%20Component%20Object%20References) (BurstCameraShakeInstance :type Camera%20Shake%20Base%20Object%20Reference) (BurstDecalInstance :type Decal%20Actor%20Object%20Reference))
  (|CallParentFunction MyTarget Parameters ParticleComponents AudioComponents BurstCameraShakeInstance BurstDecalInstance)
  (bind _asbga_bunker_door_tidal_crane (Utilities|Casting|CastToBGA_BunkerDoor_TidalCrane (Actor|GetParentActor MyTarget)))
  (Audio|SpawnSoundAttached "/MidMatchWeaponBunkerGameplay/Items/WeaponBunkerDoor/MMOBunkers_Ch6S4/Sound/Waves/MSS_TidalCrane_Bunker_Scan_Speaker_VO.MSS_TidalCrane_Bunker_Scan_Speaker_VO" (Variables|/MidMatchWeaponBunkerGameplay/Items/WeaponBunkerDoor/MMOBunkers_Ch6S4/BGA_BunkerDoor_TidalCrane.BGA_BunkerDoor_TidalCrane_C|GetSpeaker _asbga_bunker_door_tidal_crane))
  (Audio|SpawnSoundAttached "/MidMatchWeaponBunkerGameplay/Items/WeaponBunkerDoor/MMOBunkers_Ch6S4/Sound/Waves/MSS_TidalCrane_Bunker_InnerLock.MSS_TidalCrane_Bunker_InnerLock" (Variables|/MidMatchWeaponBunkerGameplay/Items/WeaponBunkerDoor/MMOBunkers_Ch6S4/BGA_BunkerDoor_TidalCrane.BGA_BunkerDoor_TidalCrane_C|GetSpeaker _asbga_bunker_door_tidal_crane)))
